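; Anim notify for the SunbeamQuest backpack cosmetic that adjusts the visibility of its spawned Niagara effect in the frontend.
; On notify, it casts the spawned effect to a Niagara component and the mesh owner to a Fort player pawn.
; When the cosmetic is displayed in a frontend context, the effect is shown unless the pawn is flagged as a frontend hologram.
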(fn Received_Notify
  (params (MeshComp :type Skeletal%20Mesh%20Component%20Object%20Reference) (Animation :type Anim%20Sequence%20Base%20Object%20Reference) (EventReference :type Anim%20Notify%20Event%20Reference%20Structure%20%28by%20ref%29))
  (bind _returnvalue (AnimNotify|GetSpawnedEffect))
  (bind _asniagara_particle_system_component (Utilities|Casting|CastToNiagaraComponent _returnvalue)
    (:then
      (bind _asfort_player_pawn (Utilities|Casting|CastToFortPlayerPawn (Components|GetOwner (Variables|GetMeshComp))))
      (if (Reactive|Frontend|IsCosmeticObjectDisplayedinFrontendContext 0)
        (Rendering|SetVisibility _asniagara_particle_system_component (not (Variables|/Script/FortniteGame.FortPlayerPawn|GetbIsInFrontEndHologram _asfort_player_pawn)) false)
        (return false)))
    (:CastFailed
      (return false))))
