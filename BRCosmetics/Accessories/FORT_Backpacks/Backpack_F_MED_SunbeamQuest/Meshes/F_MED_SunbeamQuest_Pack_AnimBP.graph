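# Animation Blueprint for the SunbeamQuest backpack cosmetic, playing idles and a kill reaction when the equipped variant enables reactivity.
# Separate frontend and in-game state machines are selected by bool; the frontend replays the reaction on a notify-started delay.
# A watched kill stat briefly pulses the in-game trigger, and both triggers timestamp a material parameter.
# Chained additive bone transforms reposition the pack using owner spine bend, movement stance and crouch-sprinting.
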AnimBlueprint /BRCosmetics/Accessories/FORT_Backpacks/Backpack_F_MED_SunbeamQuest/Meshes/F_MED_SunbeamQuest_Pack_AnimBP
target_skeleton=/Game/Accessories/FORT_Backpacks/Backpack_F_MED_SunbeamQuest/Meshes/F_MED_SunbeamQuest_Pack_Skeleton
graphs [
 AnimGraph {nodes [
   AnimGraphNode_StateMachine_0 = Frontend + Reactive() {is=title AnimGraphNode_StateMachine}
   AnimGraphNode_StateMachine_1 = Idle + Reactive() {is=title AnimGraphNode_StateMachine}
   n1 = Get bShouldDisplayFrontendReactivity() -> (bShouldDisplayFrontendReactivity:bool)
   n2 = "Transform (Modify) Bone - Bone: pack"(
    ComponentPose=Local To Component(LocalPose=Blend Poses by bool(BlendPose_0=@AnimGraphNode_StateMachine_0, BlendPose_1=@AnimGraphNode_StateMachine_1, bActiveValue=@n1))
    Translation="0,-1,10"
    Rotation="10,-15,-7"
    Scale="0.7,0.7,0.7"
   ) {
    rotationMode=BMM_Additive
    translationMode=BMM_Additive
    alphaBoolBlend {blendOption=Linear}
    boneToModify {boneName=pack}
    rotation=<rot 10 -15 -7>
    scale=<0.7 0.7 0.7>
    translation {y=-1, z=10}
   }
   n3 = "Transform (Modify) Bone - Bone: pack"(ComponentPose=@n2, Translation=Get backpack_translation(), Rotation=Get backpack_rotate()) {
    rotationMode=BMM_Additive
    translationMode=BMM_Additive
    alphaBoolBlend {blendOption=Linear}
    boneToModify {boneName=pack}
    rotation {roll=-7}
    translation {z=2}
   }
   n4 = "Transform (Modify) Bone - Bone: pack"!(ComponentPose=@n3, bAlphaBoolEnabled:bool=Get bIsCrouchSprinting(), Translation="0,-10,0", Rotation="0,0,-45", Scale) -> (Pose) {
    alphaInputType=Bool
    rotationMode=BMM_Additive
    translationMode=BMM_Additive
    alphaBoolBlend {blendInTime=0.3, blendOutTime=0.3}
    boneToModify {boneName=pack}
    rotation {roll=-45}
    translation {y=-10}
   }
   Output Pose(Result=Component To Local(ComponentPose=@n4)) {layerGroup=None}
  ]}
 Frontend + Reactive {
  entry=Idle
  path=AnimGraph.AnimGraphNode_StateMachine_0.Frontend + Reactive
  ids=id
  nodes [
   AnimStateNode_0 = Idle() {stateEntered {guid=#1, notifyName=FE Reactive}}
   AnimStateNode_1 = FE Reactive() {is=title AnimStateNode}
   AnimStateTransitionNode_0 = AnimStateTransitionNode() {crossfadeDuration=0.15}
   AnimStateTransitionNode_1 = AnimStateTransitionNode() {bAutomaticRuleBasedOnSequencePlayerInState=true, crossfadeDuration=0.65}
  ]
  links [
   {
    dst=Idle
    rule=AnimGraph.AnimGraphNode_StateMachine_0.Frontend + Reactive.AnimStateTransitionNode_1.Transition
    src=FE Reactive
    via=AnimStateTransitionNode_1
   }
   {
    dst=FE Reactive
    rule=AnimGraph.AnimGraphNode_StateMachine_0.Frontend + Reactive.AnimStateTransitionNode_0.Transition
    src=Idle
    via=AnimStateTransitionNode_0
   }
  ]
 }
 Idle {
  path=AnimGraph.AnimGraphNode_StateMachine_0.Frontend + Reactive.AnimStateNode_0.Idle
  nodes [
   n1 = Random Sequence Player!() -> (Pose:Pose Link Structure) {
    is=title AnimGraphNode_RandomPlayer
    entries [
     {
      chanceToPlay=0.65
      maxLoopCount=0
      maxPlayRate=1
      minLoopCount=0
      minPlayRate=1
      sequence=/BRCosmetics/Accessories/FORT_Backpacks/Backpack_F_MED_SunbeamQuest/Animation/SunbeamQuest_Idle
      blendIn {blendOption=HermiteCubic, blendTime=0.2, customCurve=null}
     }
     {
      chanceToPlay=0.4
      maxLoopCount=0
      maxPlayRate=1
      minLoopCount=0
      minPlayRate=1
      sequence=/BRCosmetics/Accessories/FORT_Backpacks/Backpack_F_MED_SunbeamQuest/Animation/SunbeamQuest_IdleBreak
      blendIn {blendOption=HermiteCubic, blendTime=0.1, customCurve=null}
     }
    ]
   }
   Output Animation Pose(Result=@n1)
  ]
 }
 FE Reactive {
  path=AnimGraph.AnimGraphNode_StateMachine_0.Frontend + Reactive.AnimStateNode_1.FE Reactive
  nodes [
   n1 = Sequence Player 'SunbeamQuest_KillReactive'() {bLoopAnimation=false, sequence=/BRCosmetics/Accessories/FORT_Backpacks/Backpack_F_MED_SunbeamQuest/Animation/SunbeamQuest_KillReactive}
   Output Animation Pose(Result=@n1)
  ]
 }
 Transition {
  path=AnimGraph.AnimGraphNode_StateMachine_0.Frontend + Reactive.AnimStateTransitionNode_0.Transition
  nodes [Result(bCanEnterTransition=AND Boolean(A=Get FE Reactive() -> (FE Reactive:bool), B=Get Cosmetic variant reactivity() -> (Cosmetic variant reactivity:bool)))]
 }
 Transition {
  path=AnimGraph.AnimGraphNode_StateMachine_0.Frontend + Reactive.AnimStateTransitionNode_1.Transition
  nodes [
   Result()
   NOT Boolean(A=Get Cosmetic variant reactivity())
  ]
 }
 Idle + Reactive {
  entry=Idle
  path=AnimGraph.AnimGraphNode_StateMachine_1.Idle + Reactive
  ids=id
  nodes [
   AnimStateNode_0 = Idle()
   AnimStateNode_1 = Reactive() {is=title AnimStateNode}
   AnimStateTransitionNode_0 = AnimStateTransitionNode() {crossfadeDuration=0.15}
   AnimStateTransitionNode_1 = AnimStateTransitionNode() {bAutomaticRuleBasedOnSequencePlayerInState=true, crossfadeDuration=0.65}
  ]
  links [
   {
    dst=Reactive
    rule=AnimGraph.AnimGraphNode_StateMachine_1.Idle + Reactive.AnimStateTransitionNode_0.Transition
    src=Idle
    via=AnimStateTransitionNode_0
   }
   {
    dst=Idle
    rule=AnimGraph.AnimGraphNode_StateMachine_1.Idle + Reactive.AnimStateTransitionNode_1.Transition
    src=Reactive
    via=AnimStateTransitionNode_1
   }
  ]
 }
 Idle {
  path=AnimGraph.AnimGraphNode_StateMachine_1.Idle + Reactive.AnimStateNode_0.Idle
  nodes [
   n1 = Random Sequence Player() {entries [
     {
      chanceToPlay=0.8
      maxLoopCount=0
      maxPlayRate=1
      minLoopCount=0
      minPlayRate=1
      sequence=&SunbeamQuest_Idle
      blendIn {blendOption=HermiteCubic, blendTime=2, customCurve=null}
     }
     {
      chanceToPlay=0.2
      maxLoopCount=0
      maxPlayRate=1
      minLoopCount=0
      minPlayRate=1
      sequence=&SunbeamQuest_IdleBreak
      blendIn {blendOption=HermiteCubic, blendTime=0.3, customCurve=null}
     }
    ]}
   Output Animation Pose(Result=@n1)
  ]
 }
 Reactive {
  path=AnimGraph.AnimGraphNode_StateMachine_1.Idle + Reactive.AnimStateNode_1.Reactive
  nodes [
   n1 = Sequence Player 'SunbeamQuest_KillReactive'() {bLoopAnimation=false, sequence=&SunbeamQuest_KillReactive}
   Output Animation Pose(Result=@n1)
  ]
 }
 Transition {
  path=AnimGraph.AnimGraphNode_StateMachine_1.Idle + Reactive.AnimStateTransitionNode_0.Transition
  nodes [Result(bCanEnterTransition=AND Boolean(A=Get Active_Kill() -> (Active_Kill:bool), B=Get Cosmetic variant reactivity()))]
 }
 Transition {path=AnimGraph.AnimGraphNode_StateMachine_1.Idle + Reactive.AnimStateTransitionNode_1.Transition, nodes [Result()]}
 EventGraph {nodes [
   Event BlueprintUpdateAnimation()
   > n1 = Sequence!(execute) -> (then_0, then_1, then_2:Exec)
   > Get!(execute:Exec) -> (then:Exec, else:Exec, Pawn:Fort Player Pawn Object Reference)
   > Branch(Condition=Get Cosmetic variant reactivity())
   > Do Once()
   > n2 = WatchCosmeticStat!(
    execute:Exec
    Pawn:Fort Player Pawn Object Reference=Get Pawn() -> (Pawn:Fort Player Pawn Object Reference)
    StatTag:Fort Stat Manager Tag Structure=(TagName="StatManager.AthenaKills")
    DefaultValue:int=0
    bBroadcastInitialValue:bool=true
    InitialValueBroadcastDelay:float=-1
    ChangeValueBroadcastDelay:float=-1
    bConsumeWaitingNotifications:bool=true
   ) -> (then:Exec, OnStatChanged:Exec, StatName:Name, StatValue:int) {is=title K2Node_AsyncAction}
   n3 = Set KillCount(execute=@n2.OnStatChanged, KillCount:double="To Float (Integer)"(InInt=@n2.StatValue)) -> (Output_Get:double)
   > Branch(Condition="float > float"(A=@n3, B=Get Last_KillCount() -> (Last_KillCount:double)))
   > n4 = Sequence!(execute) -> (then_0, then_1)
   > Set Last_KillCount(Last_KillCount:double=@n3) -> (Output_Get:double)
   > Set Active_Kill(Active_Kill:bool=true) -> (Output_Get:bool)
   > RetriggerableDelay(Duration=0.2)
   > Set Active_Kill(Active_Kill=false)
   Branch(execute=[@n4.then_1, @n16.then], Condition=Get Cosmetic variant reactivity())
   > SetScalarParameterValueOnMaterials!(
    execute
    self:Mesh Component Object Reference=GetPartSkeletalMeshComponent()
    ParameterName:Name=TimeFrom-BP----------------
    ParameterValue:float=GetGameTimeInSeconds()
   ) -> (then)
   n5 = Branch(
    execute=@n1.then_1
    Condition=OR Boolean!(A=Get bIsSkyDiving(), B=Get bIsMoving2D(), C:bool=Get bIsJumping(), D:bool=Get Active_Kill()) -> (ReturnValue)
   )
   > Set play idle then false(play idle then false:bool=true) -> (Output_Get:bool)
   Set play idle then false(execute=@n5.else, play idle then false=false)
   n6 = GetOwnerSkeletalMeshComponent()
   n7 = BreakRotator(InRot=BreakTransform(InTransform=GetDeltaTransformFromRefPose(self=@n6, BoneName=spine_05, BaseName=spine_01)).Rotation)
   Set backpack_rotate(execute=@n1.then_2, backpack_rotate=MakeRotator(Roll=MapRangeClamped(Value=@n7.Yaw, InRangeA=0, InRangeB=-30, OutRangeB=35)))
   n8 = Get bIsTacticalSprinting()
   n9 = Get bIsSkyDiving()
   n10 = Get bIsDBNO()
   n11 = Get bIsCrouching()
   > n12 = Branch(Condition=NOT Boolean(A=OR Boolean(A=@n9, B=@n10, C=@n11, D=@n8)))
   > Set backpack_translation(backpack_translation="0, 0, 0")
   Set backpack_translation(
    execute=@n12.else
    backpack_translation=MakeVector(
     Y=MapRangeClamped(Value=@n7.Yaw, InRangeA=0, InRangeB=30, OutRangeB=-6)
     Z=MapRangeClamped(Value=@n7.Yaw, InRangeA=0, InRangeB=-30, OutRangeB=10)
    )
   )
   Event OnCharacterCustomizationLoaded()
   > n13 = Cast To FortPlayerPawn(Object=TryGetFortPawn())
   > n14 = Set Pawn(Pawn:Fort Player Pawn Object Reference=@n13) -> (Output_Get:Fort Player Pawn Object Reference)
   > n15 = Cast To B_Athena_PartModifier_SunBeamQuest!(
    execute
    Object=GetModifierForPartType!(self:Fort Player Pawn Object Reference=@n14, PartType:EFortCustomPartType Enum=Backpack) -> (ReturnValue:Custom Character Part Modifier Object Reference)
   ) -> (then, CastFailed, AsB Athena Part Modifier Sun Beam Quest:B Athena Part Modifier Sun Beam Quest Object Reference) {is=title K2Node_DynamicCast}
   > Set Cosmetic variant reactivity(Cosmetic variant reactivity:bool=Get ReactivityEnabled!(self:B Athena Part Modifier Sun Beam Quest Object Reference=@n15) -> (ReactivityEnabled:bool)) -> (Output_Get:bool)
   Event BlueprintInitializeAnimation()
   > Set Skel_Mesh(Skel_Mesh=GetOwnerSkeletalMeshComponent())
   AnimNotify_FE Reactive() {is=title K2Node_Event}
   > Set FE Reactive(FE Reactive:bool=false) -> (Output_Get:bool)
   > Delay(Duration=8)
   > n16 = Set FE Reactive(FE Reactive=true)
  ]}
]
variables [
 Active_Kill {type=boolean}
 backpack_rotate {type=Rotator}
 backpack_translation {type=Vector}
 Cosmetic variant reactivity {type=boolean}
 Eye {type=boolean}
 FE Reactive {type=boolean}
 Gravity Override Params {type=GravityOverrideParamsStruct, default {skydive_FloatingDown {gravityOverride=<0 -300 100>}}}
 gravity_override {type=Vector}
 KillCount {type=number}
 Last_KillCount {type=number}
 Pawn {type=FortPlayerPawn}
 play idle then false {type=boolean}
 reactivity_alpha {type=number}
 Skel_Mesh {type=SkeletalMeshComponent}
]
